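# Sound cue for the looping in-world audio of Geode Oddities, classed under Fort_Gadgets.
# A quality-level root sends the first two levels to the standard looping wave and the last two to an "_opt" variant.
# Each wave passes through its own attenuation node using Geode_Oddities_Att2 settings.
# A concurrency override limits the cue to a single playing instance.
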SoundCue /Figment_S04/Athena/Sounds/Events/Geode/Oddities/Geode_Oddities_InWorldLoop_Cue2
root=QualityLevel_0
nodes [
 {
  id=Attenuation_0
  children [WavePlayer_0]
  properties {attenuationSettings=&/Geode_Oddities_Att2}
 }
 {
  id=Attenuation_2
  children [WavePlayer_1]
  properties {attenuationSettings=&Geode_Oddities_Att2}
 }
 {id=QualityLevel_0, children [Attenuation_0, Attenuation_0, Attenuation_2, Attenuation_2]}
 {id=WavePlayer_0, properties {bLooping=true, soundWaveAssetPtr=/Game/Athena/Sounds/Events/Geode/Oddities/Geode_Oddities_InWorldLoop_01}}
 {id=WavePlayer_1, properties {bLooping=true, soundWaveAssetPtr=/Game/Athena/Sounds/Events/Geode/Oddities/Geode_Oddities_InWorldLoop_01_opt}}
]
properties {
 bOverrideConcurrency=true
 maxDistance=10000
 soundClassObject=&Fort_Gadgets
 volumeMultiplier=1
 concurrencyOverrides {maxCount=1}
}
