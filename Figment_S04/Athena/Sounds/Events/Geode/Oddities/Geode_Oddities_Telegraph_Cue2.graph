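# Sound cue for the Geode Oddities telegraph event, playing a teleport sound in the gadgets sound class.
# A quality level node at the root picks the wave: the first two levels share one teleport wave and the last two share an "opt2" variant.
# An Oddities attenuation asset handles spatial falloff, and a concurrency override limits the cue to a single playing instance.
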SoundCue /Figment_S04/Athena/Sounds/Events/Geode/Oddities/Geode_Oddities_Telegraph_Cue2
root=QualityLevel_0
nodes [
 {id=QualityLevel_0, children [WavePlayer_2,WavePlayer_2,WavePlayer_0,WavePlayer_0]}
 {id=WavePlayer_0, properties {soundWaveAssetPtr=&/Geode_Oddities_Teleport_17_opt2}}
 {id=WavePlayer_2, properties {soundWaveAssetPtr=&/Geode_Oddities_Teleport_172}}
]
properties {
 attenuationSettings=&/Geode_Oddities_Att2
 bOverrideConcurrency=true
 maxDistance=10000
 soundClassObject=&Fort_Gadgets
 volumeMultiplier=1
 concurrencyOverrides {maxCount=1}
}
